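# SoundCue that loops the Holiday Noir music track from the Season 33 Battle Pass quests ChristmasNoir audio set.
# A single looping wave player feeds the cue output, with no mixing, randomization or modulation.
# It routes through the Fort_Music_POI_PSM sound class and sets a very large max distance.
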SoundCue /BattlePassS33_Quests/Audio/ChristmasNoir/S33_HolidayNoirt_Music_Loop
root=WavePlayer_0
nodes [{id=WavePlayer_0, properties {bLooping=true, soundWaveAssetPtr=&/S33_HolidayNoir_Music_Loop}}]
properties {maxDistance=2097152, soundClassObject=/Game/Sounds/FortSoundClassesAndModes/Fort_Music_POI_PSM}
